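# Niagara idle effect for the Hacker pickaxe that adds persistent, flickering glow lights to the weapon.
# BlueLights bursts two sprites, placed and angled separately by particle partition, while YellowLightMesh adds a bright yellow mesh with a cyan-blue glow color.
# Particles never expire and fade in by curve, with a shared System.Flicker value driving glow alpha and blue sprite opacity.
# The YellowLightSides and YellowLightTop sprite emitters are disabled.
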NiagaraSystem /BRCosmetics/Weapons/FORT_Melee/Pickaxe_Hacker/FX/NS_Pickaxe_Hacker_Idle
dependencies {
 usedDataInterfaces [&NiagaraDataInterfaceCurve]
 usedDynamicInputs [&MaskFloatBySpawnGroup, &MaskVectorBySpawnGroup]
 usedModules [
  &SystemState
  &self:SystemSpawnScript.SystemScriptSource.SystemScriptGraph.NiagaraNodeAssignment_0.SetVariables_#1
 ]
}
emitters {
 BlueLights {
  emitterUpdateScript [
   &EmitterState()
   &SpawnBurst_Instantaneous(Spawn Count=2, Use Spawn Probability=false, UseLoopCountLimit=false)
  ]
  particleSpawnScript [
   &PartitionParticles()
   &InitializeParticle(
    Color=<rgba 0.1627 0.2339 1 0.14>
    Lifetime=500
    Position Offset=&MaskVector(Mask Parameter=$Particles.PartitionParticles.Partition, Mask=0, Masked Vector=<0 30 52>, Passthrough Vector=<0 20 -10>)
    Sprite Rotation Angle=&MaskFloat(Mask Parameter=$Particles.PartitionParticles.Partition, Mask=0, Masked Float=35.2966, Passthrough Float=-35)
    Sprite Rotation Mode=::NewEnumerator0
    Sprite Size=/Niagara/DynamicInputs/Helpers/MaskVector2d(
     Mask Parameter:NiagaraInt32=$Emitter.SpawnGroup
     Mask:NiagaraInt32=$Particles.PartitionParticles.Partition
     Masked Vector2D:Vector2f=<0 0>
     Passthrough Vector2D:Vector2f=<0 0>
    )
    Sprite Size Mode=::NewEnumerator4
    UsePositionOffset=true
    Write Material Random=true
   )
  ]
  particleUpdateScript [
   &ParticleState(Kill Particles When Lifetime Has Elapsed=false)
   &ScaleColor(
    Scale Alpha=&FloatFromCurve(FloatCurve=DataInterface() {propertyValues {Curve {keys [<at 0 0>,<at 0.1 1>,<at 1 1>]}}})
    ScaleA=true
    ScaleRGB=true
   )
   &ScaleColor(
    Scale Alpha=&RandomRangeFloat(Minimum=0.5, Maximum=1, Evaluation Type=::NewEnumerator1)
    ScaleA=true
    ScaleRGB=true
   ) {enabled=false, module=ScaleColor001}
   /SetVariables(
    Particles.GlowCameraOffset=8
    Particles.GlowColor=&MakeLinearColorFromVectorAndFloat("Vector (RGB)"=<0.16 0.19 1>, "Float (Alpha)"=$System.Flicker)
   ) {isSetParameters=true}
   &ScaleColor(Scale Alpha=$System.Flicker, ScaleA=true, ScaleRGB=true) {module=ScaleColor002}
  ]
  rendererClasses [&NiagaraSpriteRendererProperties]
  renderers [{rendererClass=&NiagaraSpriteRendererProperties}]
 }
 YellowLightMesh {
  emitterSpawnScript [/SetVariables() {isSetParameters=true, module=SetVariables_#2}]
  emitterUpdateScript [
   &EmitterState()
   &SpawnBurst_Instantaneous(Use Spawn Probability=false, UseLoopCountLimit=false)
  ]
  particleSpawnScript [
   &PartitionParticles()
   &InitializeParticle(
    Color=<rgba 50 29.8526 5.3121 1>
    Lifetime=500
    Mesh Scale Mode=::NewEnumerator3
    Mesh Uniform Scale=1.35
    Position Offset=<0 0 -60>
    Sprite Size=<20 50>
    Sprite Size Mode=::NewEnumerator4
    UsePositionOffset=true
    Write Material Random=true
   )
  ]
  particleUpdateScript [
   &ParticleState(Kill Particles When Lifetime Has Elapsed=false)
   &ScaleColor(
    Scale Alpha=&RandomRangeFloat(Minimum=0.5, Maximum=1, Evaluation Type=::NewEnumerator1)
    ScaleA=true
    ScaleRGB=true
   ) {enabled=false, module=ScaleColor001}
   /SetVariables(
    Particles.GlowCameraOffset=5
    Particles.GlowColor=&MakeLinearColorFromVectorAndFloat("Vector (RGB)"=<0.16 0.7428 1>, "Float (Alpha)"=$System.Flicker)
   ) {isSetParameters=true, module=SetVariables_#3}
   &ScaleColor(
    Color Value To Scale=$Particles.GlowColor
    Scale Alpha=&FloatFromCurve(FloatCurve=DataInterface() {propertyValues {Curve {keys [<at 0 0>,<at 0.1 1>,<at 1 1>]}}})
    ScaleA=true
    ScaleRGB=true
   )
  ]
  rendererClasses [&NiagaraMeshRendererProperties]
  renderers [{rendererClass=&NiagaraMeshRendererProperties, rendererIndex=0}]
 }
 YellowLightSides {
  bEnabled=false
  emitterSpawnScript [/SetVariables() {isSetParameters=true, module=SetVariables_#2}]
  emitterUpdateScript [
   &EmitterState()
   &SpawnBurst_Instantaneous(Spawn Count=4, Use Spawn Probability=false, UseLoopCountLimit=false)
  ]
  particleSpawnScript [
   &PartitionParticles()
   &InitializeParticle(
    Color=<rgba 1 0.5971 0.1062 0.14>
    Lifetime=500
    Position Offset=&MaskVector(Mask Parameter=$Particles.PartitionParticles.Partition, Mask=0, Masked Vector=<25 30 10>, Passthrough Vector=<-20 30 5>)
    Sprite Size=<20 50>
    Sprite Size Mode=::NewEnumerator4
    UsePositionOffset=true
    Write Material Random=true
   )
  ]
  particleUpdateScript [
   &ParticleState(Kill Particles When Lifetime Has Elapsed=false)
   &ScaleColor(
    Scale Alpha=&FloatFromCurve(FloatCurve=DataInterface() {propertyValues {Curve {keys [<at 0 0>,<at 0.1 1>,<at 1 1>]}}})
    ScaleA=true
    ScaleRGB=true
   )
   &ScaleColor(
    Scale Alpha=&RandomRangeFloat(Minimum=0.5, Maximum=1, Evaluation Type=::NewEnumerator1)
    ScaleA=true
    ScaleRGB=true
   ) {enabled=false, module=ScaleColor001}
   /SetVariables(
    Particles.GlowCameraOffset=5
    Particles.GlowColor=&MakeLinearColorFromVectorAndFloat("Vector (RGB)"=<1 0.63 0>, "Float (Alpha)"=$System.Flicker)
   ) {isSetParameters=true, module=SetVariables_#3}
  ]
  rendererClasses [&NiagaraSpriteRendererProperties]
  renderers [{rendererClass=&NiagaraSpriteRendererProperties}]
 }
 YellowLightTop {
  bEnabled=false
  emitterUpdateScript [
   &EmitterState()
   &SpawnBurst_Instantaneous(Spawn Count=2, Use Spawn Probability=false, UseLoopCountLimit=false)
  ]
  particleSpawnScript [
   &PartitionParticles()
   &InitializeParticle(
    Color=<rgba 1 0.5971 0.1062 0.14>
    Lifetime=500
    Position Offset=&MaskVector(Mask Parameter=$Particles.PartitionParticles.Partition, Mask=0, Masked Vector=<17 27 38>, Passthrough Vector=<-18 27 38>)
    Sprite Rotation Angle=&MaskFloat(Mask Parameter=$Particles.PartitionParticles.Partition, Mask=0, Masked Float=35.2966, Passthrough Float=-35)
    Sprite Rotation Mode=::NewEnumerator0
    Sprite Size=<30 30>
    Sprite Size Mode=::NewEnumerator4
    UsePositionOffset=true
    Write Material Random=true
   )
  ]
  particleUpdateScript [
   &ParticleState(Kill Particles When Lifetime Has Elapsed=false)
   &ScaleColor(
    Scale Alpha=&FloatFromCurve(FloatCurve=DataInterface() {propertyValues {Curve {keys [<at 0 0>,<at 0.1 1>,<at 1 1>]}}})
    ScaleA=true
    ScaleRGB=true
   )
   &ScaleColor(
    Scale Alpha=&RandomRangeFloat(Minimum=0.5, Maximum=1, Evaluation Type=::NewEnumerator1)
    ScaleA=true
    ScaleRGB=true
   ) {enabled=false, module=ScaleColor001}
   &ScaleColor(Scale Alpha=$System.Flicker, ScaleA=true, ScaleRGB=true) {module=ScaleColor002}
   /SetVariables(
    Particles.GlowCameraOffset=5
    Particles.GlowColor=&MakeLinearColorFromVectorAndFloat("Vector (RGB)"=<1 0.63 0>, "Float (Alpha)"=$System.Flicker)
   ) {isSetParameters=true}
  ]
  rendererClasses [&NiagaraSpriteRendererProperties]
  renderers [{rendererClass=&NiagaraSpriteRendererProperties}]
 }
}
summary {
 systemName=NS_Pickaxe_Hacker_Idle
 emitters [
  {bEnabled=false, emitterName=YellowLightSides, rendererClasses [&NiagaraSpriteRendererProperties]}
  {bEnabled=false, emitterName=YellowLightTop, rendererClasses [&NiagaraSpriteRendererProperties]}
  {emitterName=BlueLights, rendererClasses [&NiagaraSpriteRendererProperties]}
  {emitterName=YellowLightMesh, rendererClasses [&NiagaraMeshRendererProperties]}
 ]
 userVariables []
}
