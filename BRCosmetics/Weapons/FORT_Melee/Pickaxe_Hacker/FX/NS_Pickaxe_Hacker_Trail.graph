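# Niagara system for the swing trail of the Hacker pickaxe cosmetic, drawn as four ribbon emitters.
# Each emitter spawns and updates anim-trail particles from a skeletal mesh, layering ribbons of different width, lifetime and color curves.
# The three Air_Trail layers hue-shift over life and, once the system goes Inactive, a DoOnce-triggered custom timeline drives alpha fades.
# Trail_color scales alpha by time of day; a sprite debug emitter is disabled.
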NiagaraSystem /BRCosmetics/Weapons/FORT_Melee/Pickaxe_Hacker/FX/NS_Pickaxe_Hacker_Trail
dependencies {
 usedDataInterfaces [&NiagaraDataInterfaceCurve, &NiagaraDataInterfaceColorCurve, &NiagaraDataInterfaceArrayFloat]
 usedDynamicInputs []
 usedModules [
  &self:SystemSpawnScript.SystemScriptSource.SystemScriptGraph.NiagaraNodeAssignment_0.SetVariables_#1
  &SystemState
 ]
}
emitters {
 Air_Trail {
  emitterUpdateScript [
   &EmitterState(Life Cycle Mode=::NewEnumerator1, Loop Behavior=::NewEnumerator1, Loop Delay=0.05, Loop Duration=0.3, UseLoopDelay=true)
   &FN_SpawnAnimTrail_v2()
  ]
  particleSpawnScript [
   /Game/Effects/Niagara/Modules/AnimTrail/FN_InitializeAnimTrail/FN_InitializeAnimTrailBoneIndecies(Second Bone or Socket Index=1)
   &FN_InitializeAnimTrail(AnimTrailWidth=2, SkeletalMesh=$System.Skeletal Mesh) {module=FN_InitializeAnimTrail001}
   /Game/Effects/Niagara/Modules/AnimTrail/FN_InitializeAnimTrail/FN_InitializeAnimTrail(
    AnimTrailWidth=1
    Color=<rgba 0.1059 0.7294 0.949 1>
    Lifetime=0.4
    Position Offset:Vector3f=&Lerp_Vector(A=<0 0 0>, B=<0 0 0>, Alpha=&MakeFloatFromInt(Int=$Particles.SpawnGroup))
   ) {enabled=false}
   &DynamicMaterialParameters(Module.Param0WriteEnabled=false, Module.Param1WriteEnabled=false, Module.Param2WriteEnabled=false, Module.Param3WriteEnabled=false)
  ]
  particleUpdateScript [
   &SolveForcesAndVelocity(Clamp Velocity=false, Limit Acceleration=false)
   &FN_UpdateAnimTrail(SkeletalMesh=$System.Skeletal Mesh)
   &FN_UpdateAnimTrail_v2() {enabled=false}
   &ParticleState()
   &FN_UpdateAnimTrail(Locked Particle RibbonWidth=1, SkeletalMesh=$System.Skeletal Mesh) {enabled=false, module=FN_UpdateAnimTrail001}
   &ScaleRibbonWidth(Ribbon Width Scale=&FloatFromCurve(FloatCurve=DataInterface() {propertyValues {Curve {keys [{arriveTangent=-0.8703, time=0.2, value=1}, <at 1 0.8>]}}}))
   &DoOnce(Trigger Condition:NiagaraBool=&self:ScratchDynamicInput_01(ENiagara Execution State:ENiagaraExecutionState=$System.ExecutionState, ENiagara Execution State001:ENiagaraExecutionState=::Inactive))
   &self:Custometimeline(
    Play:NiagaraBool=$Particles.DoOnce.TriggeredOnce
    Play Rate:NiagaraFloat=&Divide_Float(A=1, B=0.04)
    Rewind Mode:ENiagara_TimelineMode=::NewEnumerator1
   ) {module=Custometimeline}
   &ScaleColor(
    Scale Alpha=&FloatFromCurve(FloatCurve=DataInterface() {propertyValues {Curve {keys [{arriveTangent=19.9617, interpMode=RCIM_Cubic, tangentMode=RCTM_User, time=0, value=1}, <at 1 0>]}}})
    Scale RGB=&VectorFromFloat(Value=&FNTimeOfDayScalar(Night Scale=0.35, Morning Scale=0.45, Day Scale=0.6, Evening Scale=0.375))
    ScaleA=true
    ScaleRGB=true
   ) {enabled=false}
   &SpriteFacingAndAlignment(Sprite Facing=$Particles.RibbonFacing, WriteAlignment=false, WriteFacing=true) {enabled=false}
   &Color(
    Color=&HueShiftLinearColor(
     Hue=&Lerp_Float(
      A=0
      B=&RandomRangeFloat(Minimum=0.05, Maximum=-0.1)
      Alpha=&FloatFromCurve(FloatCurve=DataInterface() {propertyValues {Curve {keys [<at 0 0 RCIM_Cubic>, {arriveTangent=2.694, interpMode=RCIM_Cubic, leaveTangent=2.694, tangentMode=RCTM_User, time=1, value=1}]}}})
     )
     Saturation=&FloatFromCurve(FloatCurve=DataInterface() {propertyValues {Curve {keys [<at 0 1 RCIM_Cubic>, {arriveTangent=2.9139, interpMode=RCIM_Cubic, leaveTangent=2.9139, tangentMode=RCTM_User, time=1, value=2}]}}})
     Value=1
     Linear Color=&ColorFromCurve(ColorCurve=DataInterface() {propertyValues {
        AlphaCurve {keys [<at 0 1 RCIM_Cubic>,<at 1 0 RCIM_Cubic>]}
        BlueCurve {keys [{arriveTangent=-0.8091, interpMode=RCIM_Cubic, time=0.22, value=0.949}, {interpMode=RCIM_Cubic, leaveTangent=-1.1334, time=0.8356, value=1}]}
        GreenCurve {keys [
          {arriveTangent=-0.3753, interpMode=RCIM_Cubic, time=0.22, value=0.7294}
          {interpMode=RCIM_Cubic, leaveTangent=-0.3429, time=0.8356, value=0.67}
         ]}
        RedCurve {keys [<at 0.22 0.1059 RCIM_Cubic>, {interpMode=RCIM_Cubic, leaveTangent=-1.0652, time=0.8356, value=1}]}
       }})
    )
    Scale Alpha=&Multiply_Float(
     A=0.8
     B=&FloatFromCurve(
      FloatCurve=DataInterface() {propertyValues {Curve {keys [{arriveTangent=-3.0028, interpMode=RCIM_Cubic, time=0, value=1}, <at 1 0 RCIM_Cubic>]}}}
      CurveIndex=$Particles.Custometimeline.NormalizedTime
     )
    )
   )
  ]
  rendererClasses [&NiagaraRibbonRendererProperties]
  renderers [{rendererClass=&NiagaraRibbonRendererProperties}]
 }
 Air_Trail001 {
  emitterUpdateScript [
   &EmitterState(Life Cycle Mode=::NewEnumerator1, Loop Behavior=::NewEnumerator1, Loop Delay=0.05, Loop Duration=0.5, UseLoopDelay=true)
   &FN_SpawnAnimTrail_v2()
  ]
  particleSpawnScript [
   &FN_InitializeAnimTrailBoneIndecies(Second Bone or Socket Index=1)
   /Game/Effects/Niagara/Modules/AnimTrail/FN_InitializeAnimTrail/FN_InitializeAnimTrail_v5/FN_InitializeAnimTrail(
    AnimTrailWidth=&FloatFromCurve(
     FloatCurve=DataInterface() {propertyValues {Curve {keys [
         {interpMode=RCIM_Cubic, tangentMode=RCTM_User, time=0, value=0.5}
         {interpMode=RCIM_Cubic, tangentMode=RCTM_User, time=0.12, value=1}
         {tangentMode=RCTM_User, time=0.25, value=0.5}
        ]}}}
     CurveIndex=$Emitter.NormalizedLoopAge
     Scale Curve=0.2
    )
    SkeletalMesh=$System.Skeletal Mesh
   ) {module=FN_InitializeAnimTrail001}
   &DynamicMaterialParameters(
    Index 0 Param 1=1
    Index 0 Param 2=1
    Index 0 Param 3=1
    Module.Param0WriteEnabled=true
    Module.Param1WriteEnabled=true
    Module.Param2WriteEnabled=true
    Module.Param3WriteEnabled=true
   )
   &AddVelocity(Velocity=<0 0 1>)
  ]
  particleUpdateScript [
   &SolveForcesAndVelocity(Clamp Velocity=false, Limit Acceleration=false)
   &FN_UpdateAnimTrail(SkeletalMesh=$System.Skeletal Mesh)
   &ParticleState()
   &ScaleRibbonWidth(Ribbon Width Scale=&FloatFromCurve(FloatCurve=DataInterface() {propertyValues {Curve {keys [
         {arriveTangent=-0.8703, interpMode=RCIM_Cubic, leaveTangent=-0.8703, tangentMode=RCTM_User, time=0.2, value=1}
         {interpMode=RCIM_Cubic, tangentMode=RCTM_User, time=1, value=0.25}
        ]}}}))
   &DoOnce(Trigger Condition:NiagaraBool=&self:ScratchDynamicInput_02(ENiagara Execution State:ENiagaraExecutionState=$System.ExecutionState, ENiagara Execution State001:ENiagaraExecutionState=::Inactive))
   &self:Custometimeline_01(
    Play:NiagaraBool=$Particles.DoOnce.TriggeredOnce
    Play Rate:NiagaraFloat=&Divide_Float(A=1, B=0.0001)
    Rewind Mode:ENiagara_TimelineMode=::NewEnumerator1
   ) {module=Custometimeline_01}
   &Color(
    Color=&HueShiftLinearColor(
     Hue=&Lerp_Float(
      A=0
      B=&RandomRangeFloat(Minimum=0.05, Maximum=-0.1)
      Alpha=&FloatFromCurve(FloatCurve=DataInterface() {propertyValues {Curve {keys [<at 0 0 RCIM_Cubic>, {arriveTangent=2.694, interpMode=RCIM_Cubic, leaveTangent=2.694, tangentMode=RCTM_User, time=1, value=1}]}}})
     )
     Saturation=&FloatFromCurve(FloatCurve=DataInterface() {propertyValues {Curve {keys [<at 0 1 RCIM_Cubic>, {arriveTangent=2.9139, interpMode=RCIM_Cubic, leaveTangent=2.9139, tangentMode=RCTM_User, time=1, value=2}]}}})
     Value=1
     Linear Color=&ColorFromCurve(ColorCurve=DataInterface() {propertyValues {
        AlphaCurve {keys [<at 0 1 RCIM_Cubic>,<at 1 0 RCIM_Cubic>]}
        BlueCurve {keys [{arriveTangent=-0.8091, interpMode=RCIM_Cubic, time=0.22, value=5}, {interpMode=RCIM_Cubic, leaveTangent=-1.1334, time=0.8356, value=1}]}
        GreenCurve {keys [
          {arriveTangent=-0.3753, interpMode=RCIM_Cubic, time=0.22, value=3.843}
          {interpMode=RCIM_Cubic, leaveTangent=-0.3429, time=0.8356, value=0.67}
         ]}
        RedCurve {keys [<at 0.22 0.5579 RCIM_Cubic>, {interpMode=RCIM_Cubic, leaveTangent=-1.0652, time=0.8356, value=1}]}
       }})
    )
    Scale Alpha=&Multiply_Float(
     A=0.25
     B=&FloatFromCurve(
      FloatCurve=DataInterface() {propertyValues {Curve {keys [{arriveTangent=-3.0028, interpMode=RCIM_Cubic, time=0, value=1}, <at 0.5 0 RCIM_Cubic>]}}}
      CurveIndex=$Particles.Custometimeline.NormalizedTime
     )
    )
   )
   /SetVariables(Particles.RibbonFacing=$Particles.Initial.Velocity) {isSetParameters=true}
  ]
  rendererClasses [&NiagaraRibbonRendererProperties]
  renderers [{rendererClass=&NiagaraRibbonRendererProperties}]
 }
 Air_Trail002 {
  emitterUpdateScript [
   &EmitterState(Life Cycle Mode=::NewEnumerator1, Loop Behavior=::NewEnumerator1, Loop Duration=0.25, UseLoopDelay=true)
   &FN_SpawnAnimTrail_v2(AnimTrailWidthMode=::FromFirst)
  ]
  particleSpawnScript [
   &FN_InitializeAnimTrailBoneIndecies(Second Bone or Socket Index=1)
   /Game/Effects/Niagara/Modules/AnimTrail/FN_InitializeAnimTrail/FN_InitializeAnimTrail(
    AnimTrailWidth=0.1
    Color=<rgba 0.1059 0.7294 0.949 1>
    Lifetime=0.225
    Position Offset:Vector3f=&Lerp_Vector(A=<0 0 30>, B=<0 0 30>, Alpha=&MakeFloatFromInt(Int=$Particles.SpawnGroup))
   ) {enabled=false}
   /Game/Effects/Niagara/Modules/AnimTrail/FN_InitializeAnimTrail/FN_InitializeAnimTrail_v5/FN_InitializeAnimTrail(AnimTrailWidth=0.08, Lifetime=0.3, Position Offset=<5 0 0>, SkeletalMesh=$System.Skeletal Mesh) {module=FN_InitializeAnimTrail001}
   &DynamicMaterialParameters(
    Index 0 Param 1=1
    Index 0 Param 2=1
    Index 0 Param 3=1
    Module.Param0WriteEnabled=true
    Module.Param1WriteEnabled=true
    Module.Param2WriteEnabled=true
    Module.Param3WriteEnabled=true
   )
  ]
  particleUpdateScript [
   &FN_UpdateAnimTrail_v2() {enabled=false}
   &FN_UpdateAnimTrail(SkeletalMesh=$System.Skeletal Mesh)
   &ParticleState()
   &FN_UpdateAnimTrail(Locked Particle RibbonWidth=1, SkeletalMesh=$System.Skeletal Mesh) {enabled=false, module=FN_UpdateAnimTrail001}
   &ScaleRibbonWidth(Ribbon Width Scale=&FloatFromCurve(FloatCurve=DataInterface() {propertyValues {Curve {keys [
         {arriveTangent=-0.8703, interpMode=RCIM_Cubic, leaveTangent=-0.8703, tangentMode=RCTM_User, time=0.2, value=1}
         {interpMode=RCIM_Cubic, tangentMode=RCTM_User, time=1, value=0.25}
        ]}}}))
   &DoOnce(Trigger Condition:NiagaraBool=&self:ScratchDynamicInput_02(ENiagara Execution State:ENiagaraExecutionState=$System.ExecutionState, ENiagara Execution State001:ENiagaraExecutionState=::Inactive))
   &self:Custometimeline_01(
    Play:NiagaraBool=$Particles.DoOnce.TriggeredOnce
    Play Rate:NiagaraFloat=&Divide_Float(A=1, B=0.04)
    Rewind Mode:ENiagara_TimelineMode=::NewEnumerator1
   ) {module=Custometimeline_01}
   &ScaleColor(
    Scale Alpha=&FloatFromCurve(FloatCurve=DataInterface() {propertyValues {Curve {keys [{arriveTangent=19.9617, time=0, value=1}, <at 1 0>]}}}, Scale Curve=0.5)
    Scale RGB=&VectorFromFloat(Value=&FNTimeOfDayScalar(Night Scale=0.35, Morning Scale=0.45, Day Scale=0.6, Evening Scale=0.375))
    ScaleA=true
    ScaleRGB=true
   ) {enabled=false}
   &SpriteFacingAndAlignment(Sprite Facing=$Particles.RibbonFacing, WriteAlignment=false, WriteFacing=true)
   &Color(
    Color=&HueShiftLinearColor(
     Hue=&Lerp_Float(
      A=0
      B=&RandomRangeFloat(Minimum=0.05, Maximum=-0.1)
      Alpha=&FloatFromCurve(FloatCurve=DataInterface() {propertyValues {Curve {keys [<at 0 0 RCIM_Cubic>, {arriveTangent=2.694, interpMode=RCIM_Cubic, leaveTangent=2.694, tangentMode=RCTM_User, time=1, value=1}]}}})
     )
     Saturation=&FloatFromCurve(FloatCurve=DataInterface() {propertyValues {Curve {keys [<at 0 1 RCIM_Cubic>, {arriveTangent=2.9139, interpMode=RCIM_Cubic, leaveTangent=2.9139, tangentMode=RCTM_User, time=1, value=2}]}}})
     Value=1
     Linear Color=&ColorFromCurve(ColorCurve=DataInterface() {propertyValues {
        AlphaCurve {keys [<at 0 1 RCIM_Cubic>,<at 1 0 RCIM_Cubic>]}
        BlueCurve {keys [{arriveTangent=-0.8091, interpMode=RCIM_Cubic, time=0.22, value=3}, {interpMode=RCIM_Cubic, leaveTangent=-1.1334, time=0.8356, value=1}]}
        GreenCurve {keys [
          {arriveTangent=-0.3753, interpMode=RCIM_Cubic, time=0.22, value=0.2444}
          {interpMode=RCIM_Cubic, leaveTangent=-0.3429, time=0.8356, value=0.67}
         ]}
        RedCurve {keys [<at 0.22 1.2051 RCIM_Cubic>, {interpMode=RCIM_Cubic, leaveTangent=-1.0652, time=0.8356, value=1}]}
       }})
    )
    Scale Alpha=&Multiply_Float(
     A=0.25
     B=&FloatFromCurve(
      FloatCurve=DataInterface() {propertyValues {Curve {keys [{arriveTangent=-3.0028, interpMode=RCIM_Cubic, time=0, value=1}, <at 1 0 RCIM_Cubic>]}}}
      CurveIndex=$Particles.Custometimeline_01.NormalizedTime
     )
    )
    Scale Color=&VectorFromFloat(Value=&FloatFromCurve(FloatCurve=DataInterface() {propertyValues {Curve {keys [
          <at 0 0 RCIM_Cubic>
          {arriveTangent=2, interpMode=RCIM_Cubic, leaveTangent=2, time=0.1, value=5}
          {arriveTangent=-4.4529, interpMode=RCIM_Cubic, leaveTangent=-4.4529, tangentMode=RCTM_User, time=0.5, value=1}
         ]}}}))
   )
  ]
  rendererClasses [&NiagaraRibbonRendererProperties]
  renderers [{rendererClass=&NiagaraRibbonRendererProperties}]
 }
 FN_Debug {
  bEnabled=false
  emitterUpdateScript [
   &EmitterState()
   &SpawnBurst_Instantaneous(Spawn Count=1, Use Spawn Probability=false)
  ]
  particleSpawnScript [/Niagara/Modules/Spawn/Initialization/InitializeParticle(
    Lifetime=5
    Module.Write Color=true
    Module.Write Lifetime=true
    Module.Write Mass=true
    Module.Write Material Random=true
    Module.Write Position=false
    Module.Write Scale=false
    Module.Write SpriteRotation=false
    Module.Write SpriteSize=true
    Position=&SimulationPosition()
    Sprite Size=<100 100>
   )]
  particleUpdateScript [
   &ParticleState()
   &DoOnce(Trigger Condition:NiagaraBool=&self:ScratchDynamicInput_02(ENiagara Execution State:ENiagaraExecutionState=$System.ExecutionState, ENiagara Execution State001:ENiagaraExecutionState=::Inactive))
   &self:Custometimeline_01(
    Play:NiagaraBool=$Particles.DoOnce.TriggeredOnce
    Play Rate:NiagaraFloat=&Divide_Float(A=1, B=0.1)
    Rewind Mode:ENiagara_TimelineMode=::NewEnumerator1
   ) {module=Custometimeline_01}
   /SetVariables(Particles.Debug:NiagaraFloat=$Particles.Custometimeline_01.NormalizedTime) {isSetParameters=true}
   &DynamicMaterialParameters(
    Index 0 Param 1=$Particles.Debug
    Module.Param0WriteEnabled=true
    Module.Param1WriteEnabled=true
    Module.Param2WriteEnabled=true
    Module.Param3WriteEnabled=true
   )
  ]
  rendererClasses [&NiagaraSpriteRendererProperties]
  renderers [{rendererClass=&NiagaraSpriteRendererProperties}]
 }
 Trail_color {
  emitterUpdateScript [
   &EmitterState(
    Inactive Response=::NewEnumerator2
    Life Cycle Mode=::NewEnumerator1
    Loop Behavior=::NewEnumerator1
    Loop Delay=0.01
    Loop Duration=1
    UseLoopDelay=true
   )
   &SpawnRate(SpawnRate=35, Use Spawn Probability=false) {enabled=false}
   &FN_SpawnAnimTrail_v2()
  ]
  particleSpawnScript [
   &InitializeRibbon(
    Color=<rgba 1 1 1 6.5>
    Lifetime=0.25
    Module.Write Color=true
    Module.Write Lifetime=true
    Module.Write Mass=true
    Module.Write Material Random=true
    Module.Write Position=false
    Module.Write RibbonTwist=false
    Module.Write RibbonWidth=true
    Ribbon Width=50
   ) {enabled=false}
   &FN_InitializeAnimTrailBoneIndecies(Second Bone or Socket Index=1)
   &FN_InitializeAnimTrail(AnimTrailWidth=4, Lifetime=0.7, SkeletalMesh=$System.Skeletal Mesh) {module=FN_InitializeAnimTrail001}
   &SkeletalMeshLocation(
    Inherit Velocity=false
    MaskSpawnGroup=false
    Mesh Sampling Type=::NewEnumerator5
    Sampled Position Offset=<0 0 -10>
    Sampled Position Offset Coordinate Space:ENiagaraMeshTransforms=::NewEnumerator3
    Skeletal Mesh=$User.Swing_SK_Mesh
    Socket Sampling Mode=::NewEnumerator5
   ) {enabled=false}
  ]
  particleUpdateScript [
   &FN_UpdateAnimTrail_v2() {enabled=false}
   &FN_UpdateAnimTrail(SkeletalMesh=$System.Skeletal Mesh)
   &ParticleState()
   &SolveForcesAndVelocity(Clamp Velocity=false, Limit Acceleration=false)
   &ScaleColor(
    Scale Alpha=&Multiply_Float(
     A=&Multiply_Float(A=1, B=&FNTimeOfDayScalar(Night Scale=0.5, Morning Scale=0.5, Evening Scale=0.5))
     B=&FloatFromCurve(FloatCurve=DataInterface() {propertyValues {Curve {keys [
          <at 0 1 RCIM_Cubic>
          {arriveTangent=-2.0952, interpMode=RCIM_Cubic, leaveTangent=-2.0952, time=0.18, value=0.4}
          <at 0.42 0.12>
          <at 0.7723 0.0471>
          <at 0.7898 -0.0042>
          <at 1 0>
         ]}}})
    )
    Scale RGB=&VectorFromFloat(Value=&MakeFloatFromVector(
      Vector=&VectorFromFloat(Value=&FloatFromCurve(FloatCurve=DataInterface() {propertyValues {Curve {keys [
            <at 0 0 RCIM_Cubic>
            {arriveTangent=2, interpMode=RCIM_Cubic, leaveTangent=2, time=0.01, value=2}
            {arriveTangent=-4.4529, interpMode=RCIM_Cubic, leaveTangent=-4.4529, tangentMode=RCTM_User, time=0.5, value=1}
           ]}}}))
      Channel=::NewEnumerator0
     ))
    ScaleA=true
    ScaleRGB=true
   )
   &ScaleRibbonWidth(Ribbon Width Scale=&FloatFromCurve(FloatCurve=DataInterface() {propertyValues {Curve {keys [
         {arriveTangent=-0.0755, interpMode=RCIM_Cubic, leaveTangent=-0.0755, tangentMode=RCTM_User, time=0, value=1}
         {arriveTangent=-1.6226, interpMode=RCIM_Cubic, leaveTangent=-1.6226, tangentMode=RCTM_User, time=1, value=0.3}
        ]}}}))
   &ScaleColor(
    Scale Alpha=&FloatFromCurve(FloatCurve=DataInterface() {propertyValues {Curve {keys [<at 0.1 0>,<at 0.2 1>]}}}, CurveIndex=$Emitter.Age)
    ScaleA=true
    ScaleRGB=true
   ) {module=ScaleColor001}
  ]
  rendererClasses [&NiagaraRibbonRendererProperties]
  renderers [{rendererClass=&NiagaraRibbonRendererProperties}]
 }
}
summary {
 systemName=NS_Pickaxe_Hacker_Trail
 emitters [
  {emitterName=Air_Trail, rendererClasses [&NiagaraRibbonRendererProperties]}
  {emitterName=Air_Trail001, rendererClasses [&NiagaraRibbonRendererProperties]}
  {emitterName=Trail_color, rendererClasses [&NiagaraRibbonRendererProperties]}
  {emitterName=Air_Trail002, rendererClasses [&NiagaraRibbonRendererProperties]}
  {bEnabled=false, emitterName=FN_Debug, rendererClasses [&NiagaraSpriteRendererProperties]}
 ]
 userVariables [
  User.bIsFrontEnd {type=NiagaraBool}
  User.bIsFrontEndPreview {type=NiagaraBool}
  User.PhysicalSurface {type=NiagaraInt32}
  User.RibbonIDOffset {type=NiagaraInt32}
  User.SkeletalMesh {
   type=NiagaraDataInterfaceSkeletalMesh
   defaultValue {
    struct=NiagaraExt_VariableValue_DataInterface
    value {dataInterface=&self:NiagaraDataInterfaceSkeletalMesh_0}
   }
  }
 ]
}
