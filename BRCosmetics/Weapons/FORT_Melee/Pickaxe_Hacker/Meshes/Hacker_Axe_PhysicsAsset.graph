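# Physics asset for the Hacker pickaxe melee cosmetic mesh, giving the weapon a simple collision shape.
# It has a single capsule body on the Melee_TwoHanded_Handle bone, offset along the handle and rotated to fit the axe.
# With only one body there are no constraints, so the whole weapon moves as one rigid shape.
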PhysicsAsset /BRCosmetics/Weapons/FORT_Melee/Pickaxe_Hacker/Meshes/Hacker_Axe_PhysicsAsset
bodies {Melee_TwoHanded_Handle [{shapeType=Capsule, length=62.781, radius=33.4362, center=<0.8837 0.6225 21.4159>, rotation=<rot 0.292 -90 0.0097>}]}
constraints []
